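# Niagara system for an air-charge effect, filed under WealthLamb flying FX, that shows repeated wind-cone bursts attached to a skeletal mesh.
# WindCone002 loops with random durations, re-samples the skeletal mesh location each update, and drives mesh scale, alpha fade and randomized dynamic material parameters across five mesh renderers.
# Sonic_Boom_LW is enabled but empty.
# CenterGlow, GPU Sparks and the leader-follower ribbon trail emitters are disabled.
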NiagaraSystem /PistolCanary/Assets/FX/WealthLamb/Flying/NS_PC_AirCharge
dependencies {
 usedDataInterfaces [
  &NiagaraDataInterfaceColorCurve
  &NiagaraDataInterfaceCurve
  &NiagaraDataInterfaceParticleRead
  &NiagaraDataInterfaceArrayFloat3
  &NiagaraDataInterfaceSkeletalMesh
 ]
 usedDynamicInputs [&SimulationPosition]
 usedModules [
  &SystemState
  &self:SystemSpawnScript.SystemScriptSource.SystemScriptGraph.NiagaraNodeAssignment_0.SetVariables_#1
 ]
}
emitters {
 CenterGlow {
  bEnabled=false
  emitterUpdateScript [
   &EmitterState(
    Inactive Response=::NewEnumerator1
    Life Cycle Mode=::NewEnumerator1
    Loop Behavior=::NewEnumerator1
    Loop Duration Mode=::NewEnumerator1
    Reset Age On Awaken=true
    Scalability Mode=::NewEnumerator1
    UseLoopDelay=false
   )
   &SpawnBurst_Instantaneous(Spawn Count=1, Use Spawn Probability=false)
  ]
  particleSpawnScript [&InitializeParticle(
    Color=<rgba 0.0938 0.0839 0.0811 1>
    Lifetime=2
    Mesh Scale=<1.5 6 6>
    Mesh Scale Mode=::NewEnumerator4
    Sprite Rotation Mode=::NewEnumerator2
    Sprite Size Mode=::NewEnumerator3
    UsePositionOffset=false
    Write Material Random=true
   ) {module=InitializeParticle001}]
  particleUpdateScript [
   &ParticleState(Kill Particles When Lifetime Has Elapsed=false, Let Infinitely Lived Particles Die When Emitter Deactivates=true)
   &ScaleMeshSize()
   &ScaleColor(
    Scale Alpha=&FloatFromCurve(
     FloatCurve=DataInterface() {propertyValues {Curve {keys [
         {arriveTangent=4.0516, interpMode=RCIM_Cubic, leaveTangent=4.0516, tangentMode=RCTM_User, time=0, value=0}
         {arriveTangent=0.0961, interpMode=RCIM_Cubic, leaveTangent=0.0961, tangentMode=RCTM_User, time=0.25, value=1}
         {arriveTangent=0.2272, interpMode=RCIM_Cubic, leaveTangent=0.2272, tangentMode=RCTM_User, time=1, value=1}
        ]}}}
     Scale Curve=&RandomRangeFloat(Minimum=0, Maximum=1, Evaluation Type=::NewEnumerator1)
    )
    ScaleA=true
    ScaleRGB=false
   )
   &SkeletalMeshLocation(Bone Sampling Mode=::NewEnumerator4, MaskSpawnGroup=false, Skeletal Mesh=$System.Skeletal Mesh)
   &DynamicMaterialParameters(
    Index 0 Param 1=$Engine.Time
    Index 0 Param 2=&RandomRangeFloat(Minimum=0, Maximum=1, Evaluation Type=::NewEnumerator1)
    Module.Index0Float0Write=true
    Module.Index0Float1Write=true
    Module.Index0Float2Write=true
    Module.Index0Float3Write=true
   )
  ]
  rendererClasses [&NiagaraMeshRendererProperties]
  renderers [
   {rendererClass=&NiagaraMeshRendererProperties, rendererIndex=0}
   {rendererClass=&NiagaraMeshRendererProperties, rendererIndex=1}
   {rendererClass=&NiagaraMeshRendererProperties, rendererIndex=2}
  ]
 }
 Sonic_Boom_LW {rendererClasses [], renderers []}
 Sparks {
  bEnabled=false
  simTarget=GPUComputeSim
  emitterUpdateScript [
   &EmitterState(Life Cycle Mode=::NewEnumerator1, Loop Behavior=::NewEnumerator2, Loop Count=3, Loop Duration=0.2, UseLoopDelay=true)
   &SpawnBurst_Instantaneous(Spawn Count=1000, Use Spawn Probability=false, UseLoopCountLimit=false)
  ]
  particleSpawnScript [
   &InitializeParticle(
    Color Maximum=<rgba 0.3385 0.3385 0.3385 1>
    Color Minimum=<rgba 0.0677 0.0677 0.0677 1>
    Color Mode=::NewEnumerator2
    Lifetime=&RandomRangeFloat(Minimum=1, Maximum=3)
    Sprite Size Mode=::NewEnumerator3
    Uniform Sprite Size=&FloatFromCurve(
     FloatCurve=DataInterface() {propertyValues {Curve {keys [{arriveTangent=-2.2424, interpMode=RCIM_Cubic, leaveTangent=-2.2424, tangentMode=RCTM_User, time=0, value=1.5}, <at 1 0.85 RCIM_Cubic>]}}}
     CurveIndex=&RandomRangeFloat(Minimum=0, Maximum=1)
    )
    UsePositionOffset=false
    Write Material Random=true
   )
   &ShapeLocation(
    Cylinder Height=0
    Cylinder Height Midpoint=0
    Cylinder Radius=&UniformAOrBFloat(A=52.5937, B=96.188, Distribution Weight=0.412, Evaluation Type=::NewEnumerator0, Recalculate Random Each Loop=true, Override Seed=false)
    MaskSpawnGroup=false
    Override Seed=false
    Rotation Mode=::NewEnumerator2
    Shape Primitive=::NewEnumerator1
    Surface Only=true
    Use Endcaps In Surface Only Mode=false
    UseOwnerScale=true
    Yaw / Pitch / Roll=&MakeCustomVectorFromBool(
     Bool=&SetBoolByFloatComparison(A=&RandomRangeFloat(Minimum=0, Maximum=1), B=1)
     True Vector=&RandomVector(Vector Scale=1)
     False Vector=&SelectVectorFromArray(
      Array Sampling Mode=::NewEnumerator0
      Direct Array Index=?
      Vector Selection Array=DataInterface() {propertyValues {FloatData [<0.2252 0.236 0.2072>, <0.25 0.1351 0.3081>, <0.1081 0.2 0.2523>, <0 0.4595 0>]}}
      Random Seed=0
     )
    )
   )
   &AddVelocity(
    Module.Use Falloff Distance=false
    Override Seed=false
    Velocity Mode=::NewEnumerator1
    Velocity Speed=&RandomRangeFloat(Minimum=473.3125, Maximum=748.8937)
   )
   &InheritVelocity(Speed Limit=false, UseSpeedThreshold=false)
  ]
  particleUpdateScript [
   &ParticleState()
   &CurlNoiseForce(
    Bias Noise=false
    Noise Frequency=5
    Noise Strength=&FloatFromCurve(
     FloatCurve=DataInterface() {propertyValues {Curve {
        defaultValue=1
        keys [
         {arriveTangent=-3.2557, interpMode=RCIM_Cubic, leaveTangent=-3.2557, tangentMode=RCTM_User, time=0, value=1}
         {arriveTangent=-0.7893, interpMode=RCIM_Cubic, leaveTangent=-0.7893, tangentMode=RCTM_User, time=0.1, value=0.25}
         {interpMode=RCIM_Cubic, leaveTangent=-8.0089, time=1, value=0}
        ]
       }}}
     Scale Curve=6000
    )
    Pan Noise=false
   )
   &Drag(Drag=5, Use Linear Drag=true, Use Rotational Drag=true)
   &PointForce(Force Strength=-200, Module.Use Falloff=false, Module.UseLinearForce=false)
   &SolveForcesAndVelocity(Clamp Velocity=false, Limit Acceleration=false)
   &ScaleSpriteSizeBySpeed(Max Scale Factor=<1 50>, Velocity Threshold=2000)
   &ScaleColor(
    Scale Alpha=&FloatFromCurve(FloatCurve=DataInterface() {propertyValues {Curve {keys [<at 0 0>,<at 0.1 1>,<at 1 0>]}}}, Scale Curve=0.5)
    ScaleA=true
    ScaleRGB=true
   )
  ]
  rendererClasses [&NiagaraSpriteRendererProperties]
  renderers [{rendererClass=&NiagaraSpriteRendererProperties}]
 }
 Thin_Trail_Leads {
  bEnabled=false
  emitterUpdateScript [
   &EmitterState()
   &SpawnBurst_Instantaneous(Spawn Count=5, Use Spawn Probability=false, UseLoopCountLimit=false)
   &SpawnRate(SpawnRate=6, Use Spawn Probability=false)
  ]
  particleSpawnScript [
   &InitializeParticle(
    Color=&ColorFromCurve(
     ColorCurve=DataInterface() {propertyValues {
       AlphaCurve {keys [<at 0 1>,<at 1 1>]}
       BlueCurve {keys [<at 0.0052 1>,<at 0.4492 1>,<at 0.4669 0.0625>,<at 0.9983 0>]}
       GreenCurve {keys [<at 0.0052 0.02>,<at 0.4492 0.3484>,<at 0.4669 0.0625>,<at 0.9983 0>]}
       RedCurve {keys [<at 0.0052 0>,<at 0.4492 0.1235>,<at 0.4669 1>,<at 0.9983 1>]}
      }}
     CurveIndex=&RandomRangeFloat(Minimum=0, Maximum=1)
    )
    Lifetime=&RandomRangeFloat(Minimum=0.5, Maximum=1)
    UsePositionOffset=true
    Write Material Random=true
   )
   &ShapeLocation(MaskSpawnGroup=false, Override Seed=false, Sphere Radius=40, Sphere Surface Distribution=0.8214, UseOwnerScale=true)
  ]
  particleUpdateScript [
   &ParticleState()
   &PointForce(Force Strength=-4090.4429, Module.Use Falloff=false, Module.UseLinearForce=false) {enabled=false}
   &SolveForcesAndVelocity(Clamp Velocity=false, Limit Acceleration=false)
   &ScaleColor(
    Scale Alpha=&FloatFromCurve(
     FloatCurve=DataInterface() {propertyValues {Curve {
        defaultValue=0
        keys [
         {interpMode=RCIM_Cubic, tangentMode=RCTM_User, time=0, value=0}
         {interpMode=RCIM_Cubic, tangentMode=RCTM_User, time=0.5, value=1}
         {tangentMode=RCTM_User, time=1, value=0}
        ]
       }}}
     Scale Curve=0.8
    )
    Scale RGB=&VectorFromFloat(Value=&RandomRangeFloat(Minimum=1, Maximum=2, Evaluation Type=::NewEnumerator1))
    ScaleA=true
    ScaleRGB=true
   )
  ]
  rendererClasses []
  renderers []
 }
 Thin_Trails {
  bEnabled=false
  emitterUpdateScript [
   &EmitterState(
    Delay First Loop Only=true
    Inactive Response=::NewEnumerator1
    Life Cycle Mode=::NewEnumerator1
    Loop Delay=$System.Delay
    Loop Duration=1
    UseLoopDelay=true
   )
   &SpawnRate(SpawnRate=60, Use Spawn Probability=false)
   &SpawnParticlesFromOtherEmitter(
    Attribute Reader=DataInterface() {propertyValues {EmitterBinding {bindingMode=Other, emitterName=Thin_Trail_Leads}}}
    Use Maximum Age=false
    Use Minimum Age=false
   )
  ]
  particleSpawnScript [
   &InitializeParticle(Lifetime=0.75, Ribbon Width=20, Ribbon Width Mode=::NewEnumerator1, UsePositionOffset=false, Write Material Random=true)
   &SampleParticlesFromOtherEmitter(Color Sampling=::NewEnumerator0, Particle ID Sampling=::NewEnumerator1)
  ]
  particleUpdateScript [
   &ParticleState()
   &ScaleColor(
    Scale Alpha=&FloatFromCurve(FloatCurve=DataInterface() {propertyValues {Curve {keys [
         <at 0 0 RCIM_Cubic>
         {arriveTangent=-0.1027, interpMode=RCIM_Cubic, leaveTangent=-0.1027, tangentMode=RCTM_User, time=0.1, value=1}
         {arriveTangent=-1.1111, interpMode=RCIM_Cubic, leaveTangent=-1.1111, time=0.3762, value=0.3007}
         <at 1 0 RCIM_Cubic>
        ]}}})
    Scale RGB=&VectorFromFloat(Value=1)
    ScaleA=true
    ScaleRGB=true
   )
   &SolveForcesAndVelocity(Clamp Velocity=false, Limit Acceleration=false)
  ]
  rendererClasses [&NiagaraRibbonRendererProperties]
  renderers [{rendererClass=&NiagaraRibbonRendererProperties}]
 }
 WindCone002 {
  emitterUpdateScript [
   &EmitterState(
    Delay First Loop Only=true
    Inactive Response=::NewEnumerator1
    Life Cycle Mode=::NewEnumerator1
    Loop Delay=0.35
    Loop Duration=&RandomRangeFloat(Minimum=0.2, Maximum=0.4)
    Recalculate Duration Each Loop=true
    UseLoopDelay=true
   )
   &SpawnBurst_Instantaneous(Use Spawn Probability=false, UseLoopCountLimit=false)
  ]
  particleSpawnScript [&InitializeParticle(
    Color Mode=::NewEnumerator0
    Lifetime=&RandomRangeFloat(Minimum=0.6, Maximum=0.4)
    Mesh Scale Mode=::NewEnumerator4
    Position Mode=::NewEnumerator0
    Write Material Random=true
   )]
  particleUpdateScript [
   &ParticleState()
   &ScaleMeshSize(Scale Factor=&VectorFromFloat(Value=&FloatFromCurve(FloatCurve=DataInterface() {propertyValues {Curve {keys [
          <at 0 0 RCIM_Cubic>
          {arriveTangent=0.6338, interpMode=RCIM_Cubic, leaveTangent=0.6338, tangentMode=RCTM_User, time=0.2, value=0.8}
          <at 1 1 RCIM_Cubic>
         ]}}})))
   &ScaleColor(
    Scale Alpha=&FloatFromCurve(
     FloatCurve=DataInterface() {propertyValues {Curve {keys [
         <at 0 0 RCIM_Cubic>
         {arriveTangent=-1.2369, interpMode=RCIM_Cubic, leaveTangent=-1.2369, tangentMode=RCTM_User, time=0.3, value=0.45}
         <at 1 0 RCIM_Cubic>
        ]}}}
     Scale Curve=&RandomRangeFloat(Minimum=0.5, Maximum=1, Evaluation Type=::NewEnumerator1)
    )
    ScaleA=true
    ScaleRGB=true
   )
   &SkeletalMeshLocation(Bone Sampling Mode=::NewEnumerator4, MaskSpawnGroup=false, Skeletal Mesh=$System.Skeletal Mesh)
   &DynamicMaterialParameters(
    Index 0 Param 1=&Add_Float(
     A=&RandomRangeFloat(Minimum=0, Maximum=999)
     B=&FloatFromCurve(
      FloatCurve=DataInterface() {propertyValues {Curve {
         defaultValue=1
         keys [{arriveTangent=2.0581, interpMode=RCIM_Cubic, leaveTangent=2.0581, tangentMode=RCTM_User, time=0, value=0}, <at 1 1 RCIM_Cubic>]
        }}}
      Scale Curve=2
     )
    )
    Index 0 Param 2=10
    Module.Index0Param0Write=true
    Module.Index0Param1Write=true
    Module.Index0Param2Write=true
    Module.Index0Param3Write=true
   )
   &CameraOffset(Camera Offset Amount=50)
  ]
  rendererClasses [&NiagaraMeshRendererProperties]
  renderers [
   {rendererClass=&NiagaraMeshRendererProperties, rendererIndex=0}
   {rendererClass=&NiagaraMeshRendererProperties, rendererIndex=1}
   {rendererClass=&NiagaraMeshRendererProperties, rendererIndex=2}
   {rendererClass=&NiagaraMeshRendererProperties, rendererIndex=3}
   {rendererClass=&NiagaraMeshRendererProperties, rendererIndex=4}
  ]
 }
}
summary {
 systemName=NS_PC_AirCharge
 emitters [
  {bEnabled=false, emitterName=Thin_Trail_Leads, rendererClasses []}
  {bEnabled=false, emitterName=Thin_Trails, rendererClasses [&NiagaraRibbonRendererProperties]}
  {emitterName=WindCone002, rendererClasses [&NiagaraMeshRendererProperties]}
  {bEnabled=false, emitterName=Sparks, simTarget=GPUComputeSim, rendererClasses [&NiagaraSpriteRendererProperties]}
  {bEnabled=false, emitterName=CenterGlow, rendererClasses [&NiagaraMeshRendererProperties]}
  {emitterName=Sonic_Boom_LW, rendererClasses []}
 ]
 userVariables []
}
