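# Niagara flight trail for PistolCanary: a contrail that spawns by distance traveled at a skeletal mesh bone/socket.
# MainTrail drives ribbon and sprite renderers, shaping width and fading alpha over life, with its RGB color blend switched off.
# User parameters scale lifetime and ribbon width and feed glow brightness into dynamic material parameters.
# Disabled LimbTrails emitter partitions particles into four thinner ribbons on separate bones/sockets, tinted by partition and blending to gray.
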NiagaraSystem /PistolCanary/Assets/FX/WealthLamb/Flying/NS_PistolCanary_Flight_Trail
dependencies {
 usedDataInterfaces [&NiagaraDataInterfaceCurve, &NiagaraDataInterfaceSkeletalMesh]
 usedDynamicInputs []
 usedModules [
  &self:SystemSpawnScript.SystemScriptSource.SystemScriptGraph.NiagaraNodeAssignment_0.SetVariables_#1
  &SystemState
 ]
}
emitters {
 LimbTrails {
  bEnabled=false
  emitterUpdateScript [
   &EmitterState()
   &SpawnPerUnit(
    Module.Use Max Movement Threshold=false
    Module.Use Movement Tolerance=false
    Spawn Probability=1
    Spawn Spacing=10
    Use Spawn Probability=true
   )
  ]
  particleSpawnScript [
   &PartitionParticles(Number Of Partitions=4, Write Partition to Ribbon ID=::NewEnumerator0)
   &InitializeParticle(
    Lifetime=&Multiply_Float(A=0.17, B=$User.LifetimeScale)
    Ribbon Width=&Multiply_Float(A=15, B=$User.ContrailWidth)
    Ribbon Width Mode=::NewEnumerator1
    Sprite Size Mode=::NewEnumerator3
    Uniform Sprite Size=5
    UsePositionOffset=false
    Write Material Random=true
   )
   &SkeletalMeshLocation(
    Bone / Socket Index=$Particles.PartitionParticles.Partition
    Bone Sampling Mode=::NewEnumerator4
    MaskSpawnGroup=false
    Skeletal Mesh=$System.Skeletal Mesh
   )
  ]
  particleUpdateScript [
   &ParticleState()
   &DynamicMaterialParameters(
    Index 0 Param 3=&FloatFromCurve(FloatCurve=DataInterface() {propertyValues {Curve {
        defaultValue=1
        keys [
         {arriveTangent=1.3433, interpMode=RCIM_Cubic, leaveTangent=1.3433, tangentMode=RCTM_User, time=0, value=0}
         {tangentMode=RCTM_User, time=1, value=0.7}
        ]
       }}})
    Module.Index0Float0Write=false
    Module.Index0Float1Write=false
    Module.Index0Float2Write=true
    Module.Index0Float3Write=false
   )
   &ScaleRibbonWidth(Ribbon Width Scale=&FloatFromCurve(FloatCurve=DataInterface() {propertyValues {Curve {keys [<at 0 1>,<at 1 0.5>]}}}))
   &ScaleColor(
    Scale Alpha=&FloatFromCurve(FloatCurve=DataInterface() {propertyValues {Curve {keys [{arriveTangent=-2.3857, interpMode=RCIM_Cubic, leaveTangent=-2.3857, tangentMode=RCTM_User, time=0, value=1}, <at 1 0>]}}})
    Scale RGB=&MakeVectorFromLinearColorRGB(Color=&LerpLinearColors(
      StartColor=&MakeCustomLinearColorFromBool(
       Bool=&SetBoolByIntComparison(A=$Particles.PartitionParticles.Partition, B=1, Comparison Type=::NewEnumerator0)
       True Color=<rgba 0.5365 0.0149 0 1>
       False Color=<rgba 0.4896 0.247 0.1509 1>
      )
      EndColor=<rgba 0.3177 0.3177 0.3177 1>
      LerpFactor=&FloatFromCurve(FloatCurve=DataInterface() {propertyValues {Curve {keys [{interpMode=RCIM_Cubic, leaveTangentWeight=0.2, tangentMode=RCTM_User, time=0, value=0}, {tangentMode=RCTM_User, time=0.6, value=1}]}}})
     ))
    ScaleA=true
    ScaleRGB=true
   )
  ]
  rendererClasses [&NiagaraSpriteRendererProperties, &NiagaraRibbonRendererProperties]
  renderers [
   {rendererClass=&NiagaraSpriteRendererProperties}
   {rendererClass=&NiagaraRibbonRendererProperties, rendererIndex=1}
  ]
 }
 MainTrail {
  emitterUpdateScript [
   &EmitterState()
   &SpawnPerUnit(
    Module.Use Max Movement Threshold=false
    Module.Use Movement Tolerance=false
    Spawn Probability=1
    Spawn Spacing=24
    Use Spawn Probability=true
   )
  ]
  particleSpawnScript [
   &InitializeParticle(
    Lifetime=&Multiply_Float(A=0.25, B=$User.LifetimeScale)
    Ribbon Width=&Multiply_Float(A=75, B=$User.ContrailWidth)
    Ribbon Width Mode=::NewEnumerator1
    Sprite Size Mode=::NewEnumerator3
    Uniform Sprite Size=5
    UsePositionOffset=false
    Write Material Random=true
   )
   &SkeletalMeshLocation(Bone / Socket Index=4, Bone Sampling Mode=::NewEnumerator4, MaskSpawnGroup=false, Skeletal Mesh=$System.Skeletal Mesh)
  ]
  particleUpdateScript [
   &ParticleState()
   &DynamicMaterialParameters(
    Index 0 Param 2=$User.GlowBrightness
    Index 0 Param 3=&FloatFromCurve(FloatCurve=DataInterface() {propertyValues {Curve {
        defaultValue=1
        keys [
         {arriveTangent=1.3433, interpMode=RCIM_Cubic, leaveTangent=1.3433, tangentMode=RCTM_User, time=0, value=0}
         {tangentMode=RCTM_User, time=1, value=0.7}
        ]
       }}})
    Module.Index0Float0Write=false
    Module.Index0Float1Write=true
    Module.Index0Float2Write=true
    Module.Index0Float3Write=false
   )
   &ScaleRibbonWidth(Ribbon Width Scale=&FloatFromCurve(
     FloatCurve=DataInterface() {propertyValues {Curve {keys [
         {arriveTangent=20.4849, interpMode=RCIM_Cubic, leaveTangent=20.4849, tangentMode=RCTM_User, time=0, value=0.5}
         {arriveTangent=-0.1, interpMode=RCIM_Cubic, leaveTangent=-0.1, time=0.05, value=1}
         <at 1 0.4>
        ]}}}
     Scale Curve=3
    ))
   &ScaleColor(
    Scale Alpha=&FloatFromCurve(FloatCurve=DataInterface() {propertyValues {Curve {keys [{arriveTangent=-2.0738, interpMode=RCIM_Cubic, leaveTangent=-2.0738, tangentMode=RCTM_User, time=0, value=1}, <at 1 0>]}}})
    Scale RGB=&MakeVectorFromLinearColorRGB(Color=&LerpLinearColors(
      StartColor=<rgba 0.5365 0.0149 0 1>
      EndColor=<rgba 0.3177 0.3177 0.3177 1>
      LerpFactor=&FloatFromCurve(FloatCurve=DataInterface() {propertyValues {Curve {keys [{interpMode=RCIM_Cubic, leaveTangentWeight=0.2, tangentMode=RCTM_User, time=0.25, value=0}, {tangentMode=RCTM_User, time=0.6, value=1}]}}})
     ))
    ScaleA=true
    ScaleRGB=false
   )
  ]
  rendererClasses [&NiagaraSpriteRendererProperties, &NiagaraRibbonRendererProperties]
  renderers [
   {rendererClass=&NiagaraSpriteRendererProperties}
   {rendererClass=&NiagaraRibbonRendererProperties, rendererIndex=1}
  ]
 }
}
summary {
 systemName=NS_PistolCanary_Flight_Trail
 emitters [
  {
   bEnabled=false
   emitterName=LimbTrails
   rendererClasses [&NiagaraSpriteRendererProperties, &NiagaraRibbonRendererProperties]
  }
  {emitterName=MainTrail, rendererClasses [&NiagaraSpriteRendererProperties, &NiagaraRibbonRendererProperties]}
 ]
 userVariables [
  User.ContrailWidth {type=NiagaraFloat, defaultValue=5, description=""}
  User.GlowBrightness {type=NiagaraFloat, defaultValue=1, description=""}
  User.LifetimeScale {type=NiagaraFloat, defaultValue=1, description=""}
 ]
}
